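# Anim Blueprint for the F_MED_Eternity_Crown cosmetic body that makes its leg chunk bones float, with some also spinning.
# It copies the attached parent mesh's pose, then chains additive bone-space Modify Bone nodes over right and left dyn_leg_chunk bones.
# Each update derives a roll/yaw spin from real time and two bob offsets: sines of real time (A) and cosines of game time (B).
# The A and B offsets alternate along each leg to desynchronise neighbouring chunks.
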AnimBlueprint /BRCosmetics/Characters/Player/Female/Medium/Bodies/F_MED_Eternity_Crown/Meshes/F_MED_Eternity_Crown_AnimBP
target_skeleton=&/F_MED_Eternity_Crown_Skeleton
graphs [
 AnimGraph {nodes [
   n1 = Get Spinning_Rotator() -> (Spinning_Rotator:Rotator)
   n2 = Get Chunk_Float_A() -> (Chunk_Float_A:Vector)
   n3 = "Transform (Modify) Bone - Bone: dyn_leg_chunk_r_01"(
    ComponentPose=Local To Component(LocalPose=Copy Pose From Mesh() {bUseAttachedParent=true})
    Alpha=1
    Translation=@n2
    Rotation="0,0,0"
    Scale="1,1,1"
   ) {
    rotationMode=BMM_Additive
    rotationSpace=BCS_BoneSpace
    translationMode=BMM_Additive
    translationSpace=BCS_BoneSpace
    alphaBoolBlend {blendOption=Linear}
    boneToModify {boneName=dyn_leg_chunk_r_01}
   }
   n4 = Get Chunk_Float_B() -> (Chunk_Float_B:Vector)
   n5 = Get Spinning_Rotator()
   n6 = "Transform (Modify) Bone - Bone: dyn_leg_chunk_r_02"(ComponentPose=@n3, Alpha=1, Translation=@n4, Rotation=@n5, Scale="1,1,1") {
    rotationMode=BMM_Additive
    rotationSpace=BCS_BoneSpace
    translationMode=BMM_Additive
    translationSpace=BCS_BoneSpace
    alphaBoolBlend {blendOption=Linear}
    boneToModify {boneName=dyn_leg_chunk_r_02}
   }
   n7 = "Transform (Modify) Bone - Bone: dyn_leg_chunk_r_03"(ComponentPose=@n6, Alpha=1, Translation=@n2, Rotation=@n5, Scale="1,1,1") {
    rotationMode=BMM_Additive
    rotationSpace=BCS_BoneSpace
    translationMode=BMM_Additive
    translationSpace=BCS_BoneSpace
    alphaBoolBlend {blendOption=Linear}
    boneToModify {boneName=dyn_leg_chunk_r_03}
   }
   n8 = "Transform (Modify) Bone - Bone: dyn_leg_chunk_r_04"(ComponentPose=@n7, Alpha=1, Translation=@n4, Rotation=@n5, Scale="1,1,1") {
    rotationMode=BMM_Additive
    rotationSpace=BCS_BoneSpace
    translationMode=BMM_Additive
    translationSpace=BCS_BoneSpace
    alphaBoolBlend {blendOption=Linear}
    boneToModify {boneName=dyn_leg_chunk_r_04}
   }
   n9 = "Transform (Modify) Bone - Bone: dyn_leg_chunk_r_05"(ComponentPose=@n8, Alpha=1, Translation=@n2, Rotation=@n5, Scale="1,1,1") {
    translationMode=BMM_Additive
    translationSpace=BCS_BoneSpace
    alphaBoolBlend {blendOption=Linear}
    boneToModify {boneName=dyn_leg_chunk_r_05}
   }
   n10 = "Transform (Modify) Bone - Bone: dyn_leg_chunk_r_06"(ComponentPose=@n9, Alpha=1, Translation=@n4, Rotation="0,0,0", Scale="1,1,1") {
    translationMode=BMM_Additive
    translationSpace=BCS_BoneSpace
    alphaBoolBlend {blendOption=Linear}
    boneToModify {boneName=dyn_leg_chunk_r_06}
   }
   n11 = Get Chunk_Float_A()
   n12 = "Transform (Modify) Bone - Bone: dyn_leg_chunk_l_01"(ComponentPose=@n10, Alpha=1, Translation=@n11, Rotation=@n1, Scale="1,1,1") {
    rotationMode=BMM_Additive
    rotationSpace=BCS_BoneSpace
    translationMode=BMM_Additive
    translationSpace=BCS_BoneSpace
    alphaBoolBlend {blendOption=Linear}
    boneToModify {boneName=dyn_leg_chunk_l_01}
   }
   n13 = Get Chunk_Float_B()
   n14 = "Transform (Modify) Bone - Bone: dyn_leg_chunk_l_02"(ComponentPose=@n12, Alpha=1, Translation=@n13, Rotation=@n1, Scale="1,1,1") {
    rotationMode=BMM_Additive
    rotationSpace=BCS_BoneSpace
    translationMode=BMM_Additive
    translationSpace=BCS_BoneSpace
    alphaBoolBlend {blendOption=Linear}
    boneToModify {boneName=dyn_leg_chunk_l_02}
   }
   n15 = "Transform (Modify) Bone - Bone: dyn_leg_chunk_l_03"(ComponentPose=@n14, Alpha=1, Translation=@n11, Rotation=@n1, Scale="1,1,1") {
    rotationMode=BMM_Additive
    rotationSpace=BCS_BoneSpace
    translationMode=BMM_Additive
    translationSpace=BCS_BoneSpace
    alphaBoolBlend {blendOption=Linear}
    boneToModify {boneName=dyn_leg_chunk_l_03}
   }
   n16 = "Transform (Modify) Bone - Bone: dyn_leg_chunk_l_04"(ComponentPose=@n15, Alpha=1, Translation=@n13, Rotation=@n1, Scale="1,1,1") {
    rotationMode=BMM_Additive
    rotationSpace=BCS_BoneSpace
    translationMode=BMM_Additive
    translationSpace=BCS_BoneSpace
    alphaBoolBlend {blendOption=Linear}
    boneToModify {boneName=dyn_leg_chunk_l_04}
   }
   n17 = "Transform (Modify) Bone - Bone: dyn_leg_chunk_l_05"(ComponentPose=@n16, Alpha=1, Translation=@n11, Rotation="0,0,0", Scale="1,1,1") {
    translationMode=BMM_Additive
    translationSpace=BCS_BoneSpace
    alphaBoolBlend {blendOption=Linear}
    boneToModify {boneName=dyn_leg_chunk_l_05}
   }
   n18 = "Transform (Modify) Bone - Bone: dyn_leg_chunk_l_06"(ComponentPose=@n17, Alpha=1, Translation=@n13, Rotation="0,0,0", Scale="1,1,1") {
    translationMode=BMM_Additive
    translationSpace=BCS_BoneSpace
    alphaBoolBlend {blendOption=Linear}
    boneToModify {boneName=dyn_leg_chunk_l_06}
   }
   n19 = "Transform (Modify) Bone - Bone: dyn_leg_chunk_l_07"(ComponentPose=@n18, Alpha=1, Translation=@n11, Rotation="0,0,0", Scale="1,1,1") {
    translationMode=BMM_Additive
    translationSpace=BCS_BoneSpace
    alphaBoolBlend {blendOption=Linear}
    boneToModify {boneName=dyn_leg_chunk_l_07}
   }
   n20 = "Transform (Modify) Bone - Bone: dyn_leg_chunk_l_08"(ComponentPose=@n19, Alpha=1, Translation=@n13, Rotation="0,0,0", Scale="1,1,1") {
    translationMode=BMM_Additive
    translationSpace=BCS_BoneSpace
    alphaBoolBlend {blendOption=Linear}
    boneToModify {boneName=dyn_leg_chunk_l_08}
   }
   n21 = "Transform (Modify) Bone - Bone: dyn_leg_chunk_l_10"(ComponentPose=@n20, Alpha=1, Translation=@n11, Rotation=@n1, Scale="1,1,1") {
    translationMode=BMM_Additive
    translationSpace=BCS_BoneSpace
    alphaBoolBlend {blendOption=Linear}
    boneToModify {boneName=dyn_leg_chunk_l_10}
   }
   Output Pose(Result=Component To Local(ComponentPose=@n21)) {layerGroup=None}
  ]}
 EventGraph {nodes [
   Event BlueprintUpdateAnimation()
   > n1 = Sequence()
   n2 = float * float(A=float / float(A=GetRealTimeSeconds(), B=-720) {is=title K2Node_CallFunction}, B=30000)
   > Set Spinning_Rotator(Spinning_Rotator:Rotator=MakeRotator(Roll=@n2, Yaw=@n2)) -> (Output_Get:Rotator)
   n3 = GetRealTimeSeconds()
   n4 = "Sin (Radians)"(A=float * float(A=float * float(A=1, B=1), B=@n3))
   n5 = "Sin (Radians)"(A=float * float(A=float * float(A=1, B=2), B=@n3))
   n6 = "Sin (Radians)"(A=float * float(A=float * float(A=1, B=1.5), B=@n3))
   Set Chunk_Float_A(
    execute=@n1.then_1
    Chunk_Float_A:Vector=MakeVector(X=float * float(A=@n6, B=0.3), Y=float * float(A=@n4, B=0.56), Z=float * float(A=@n5, B=0.22))
   ) -> (Output_Get:Vector)
   n7 = GetGameTimeInSeconds()
   n8 = "Cos (Radians)"!(A:double=float * float(A=float * float(A=1, B=1.5), B=@n7)) -> (ReturnValue:double)
   n9 = "Cos (Radians)"(A=float * float(A=float * float(A=1, B=1.78), B=@n7))
   n10 = "Cos (Radians)"(A=float * float(A=float * float(A=1, B=1.15), B=@n7))
   > Set Chunk_Float_B(Chunk_Float_B:Vector=MakeVector(X=float * float(A=@n8, B=0.6), Y=float * float(A=@n9, B=0.725), Z=float * float(A=@n10, B=0.3))) -> (Output_Get:Vector)
  ]}
]
variables [
 Chunk_Float_A {type=Vector}
 Chunk_Float_B {type=Vector}
 Spinning_Rotator {type=Rotator}
]
