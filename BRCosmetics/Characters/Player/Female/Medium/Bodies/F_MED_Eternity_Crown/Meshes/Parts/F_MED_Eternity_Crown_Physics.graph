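# Physics asset for the crown part of the Eternity female medium player cosmetic, defining bodies for four crown pieces (top and bottom, left and right) plus the head.
# Each piece has a box on its crown parent bone, constrained to a small sphere on the matching base bone.
# Constraints lock twist and one swing axis, leaving a narrow limited swing on the other, like a tight hinge.
# The head capsule has no constraints.
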PhysicsAsset /BRCosmetics/Characters/Player/Female/Medium/Bodies/F_MED_Eternity_Crown/Meshes/Parts/F_MED_Eternity_Crown_Physics
bodies {
 dyn_crown_base_bottom_l [{shapeType=Sphere, radius=0.5}]
 dyn_crown_base_bottom_r [{shapeType=Sphere, radius=0.5}]
 dyn_crown_base_top_l [{shapeType=Sphere, radius=0.5}]
 dyn_crown_base_top_r [{shapeType=Sphere, radius=0.5}]
 dyn_crown_parent_bottom_l [{shapeType=Box, extentX=7.9546, extentY=5.9349, extentZ=1.7025, center=<3.7739 0 -0.0002>}]
 dyn_crown_parent_bottom_r [{shapeType=Box, extentX=8.4546, extentY=7.4349, extentZ=1.7025, center=<4.5644 0 -0.0001>}]
 dyn_crown_parent_top_l [{shapeType=Box, extentX=8.1847, extentY=6.3144, extentZ=1.4371, center=<4.85 -0.0003 0.0001>}]
 dyn_crown_parent_top_r [{shapeType=Box, extentX=7.6847, extentY=6.3144, extentZ=1.4371, center=<4.4087 -0.0003 0>}]
 head [{shapeType=Capsule, length=12.5, radius=6.5, center=<1.1671 -0.0001 0>, rotation=<rot 90 0 0>}]
}
constraints [
 {
  bone1Name=dyn_crown_parent_top_l
  bone2Name=dyn_crown_base_top_l
  swing1LimitDegrees=15
  swing1Motion=Locked
  swing2LimitDegrees=5
  twistLimitDegrees=0
  twistMotion=Locked
 }
 {
  bone1Name=dyn_crown_parent_bottom_l
  bone2Name=dyn_crown_base_bottom_l
  swing1LimitDegrees=15
  swing1Motion=Locked
  swing2LimitDegrees=5
  twistLimitDegrees=0
  twistMotion=Locked
 }
 {
  bone1Name=dyn_crown_parent_bottom_r
  bone2Name=dyn_crown_base_bottom_r
  swing1LimitDegrees=15
  swing1Motion=Locked
  swing2LimitDegrees=5
  twistLimitDegrees=0
  twistMotion=Locked
 }
 {
  bone1Name=dyn_crown_parent_top_r
  bone2Name=dyn_crown_base_top_r
  swing1LimitDegrees=15
  swing1Motion=Locked
  swing2LimitDegrees=5
  twistLimitDegrees=0
  twistMotion=Locked
 }
]
